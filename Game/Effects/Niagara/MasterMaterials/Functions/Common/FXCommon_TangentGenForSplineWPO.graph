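# Niagara FX material function that scales a tangent-space WPO direction by strength and rebuilds a surface tangent for spline-like displacement.
# It subtracts the current offset from the T+1 offset, adds an axis step inversely scaled by strength, move delta and frequency, then normalizes and transforms it.
# A static U/V switch picks the step axis and which tangent gets replaced; TangentBasis supplies the other.
# Outputs are both tangents and the transformed WPO.
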MaterialFunction /Game/Effects/Niagara/MasterMaterials/Functions/Common/FXCommon_TangentGenForSplineWPO
graph {nodes [
  n1 = MaterialFunctionCall() {materialFunction=&TangentBasis}
  n2 = FunctionInput() {bUsePreviewValueAsDefault=true, inputName=WPOStrength, inputType=FunctionInput_Scalar, sortPriority=2, previewValue=<1 0 0 1>}
  n3 = Constant3Vector() {constant=<rgba 0 1 0 0>}
  n4 = Constant3Vector() {constant=<rgba 1 0 0 0>}
  n5 = FunctionInput(Preview=StaticBool() {value=true}) {
   bUsePreviewValueAsDefault=true
   description="Defaults to True (expands U)"
   inputName=Expand U or V UV Channel
   inputType=FunctionInput_StaticBool
   sortPriority=6
  }
  n6 = FunctionInput() {
   bUsePreviewValueAsDefault=true
   inputName="CalcMoveDelta(TheSignIsUVFlowDirection)"
   inputType=FunctionInput_Scalar
   sortPriority=3
   previewValue=<1 0 0 1>
  }
  n7 = FunctionInput() {bUsePreviewValueAsDefault=true, inputName=AverageFrequency, inputType=FunctionInput_Scalar, sortPriority=3, previewValue=<1 0 0 1>}
  n8 = Multiply(A=Multiply(A=Divide(A=@n2) {constB=100}, B=@n6), B=@n7)
  n9 = Add(
   A=Multiply(A=FunctionInput() {bUsePreviewValueAsDefault=true, inputName=WPODirectionInT+1_TangentSpace, sortPriority=1}, B=@n2)
   B=Multiply(A=StaticSwitch(True=@n3, False=@n4, Value=@n5), B=Divide(B=@n8))
  )
  n10 = Multiply(A=FunctionInput() {bUsePreviewValueAsDefault=true, inputName=WPODirectionIn_TangentSpace}, B=@n2)
  n11 = Transform(None=Normalize(VectorInput=Subtract(A=@n9, B=@n10)))
  FunctionOutput(None=StaticSwitch(True=@n1.X U, False=@n11, Value=@n5)) {outputName=ResultTangetU}
  FunctionOutput(None=StaticSwitch(True=@n11, False=@n1.Y V, Value=@n5)) {outputName=ResultTangetV, sortPriority=1}
  FunctionOutput(None=Transform(None=@n10)) {outputName=WPO, sortPriority=2}
 ]}
