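# Material function in the Niagara FX common library that returns a UV for sampling a curve atlas to remap color.
# The greyscale input becomes U, and V is the curve index plus a half offset divided by the curve count.
# The half offset places each sample at the center of its curve's row in the atlas.
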MaterialFunction /Game/Effects/Niagara/MasterMaterials/Functions/Common/FXCommon_CurveAtlasUV
graph {
 properties {bExposeToLibrary=true, description=Return a UV for curve atlas sampling to remap color.}
 nodes [
  n1 = Add(A=Constant() {r=0.5}, B=FunctionInput() {bUsePreviewValueAsDefault=true, inputName=CurveIndex, inputType=FunctionInput_Scalar}) {constB=2}
  n2 = FunctionInput() {bUsePreviewValueAsDefault=true, inputName=CurveNum, inputType=FunctionInput_Scalar, previewValue=<1 0 0 1>}
  FunctionOutput(None=AppendVector(A=FunctionInput() {bUsePreviewValueAsDefault=true, inputName=GreyMapIn, inputType=FunctionInput_Scalar}, B=Divide(A=@n1, B=@n2) {constB=3})) {outputName=UV}
 ]
}
