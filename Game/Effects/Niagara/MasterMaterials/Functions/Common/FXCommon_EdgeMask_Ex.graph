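# Material function in the Niagara FX master materials' common set that builds a soft edge mask along one UV axis.
# The UorV scalar and its one-minus each pass through add, multiply, saturate and power, so both ends are shaped independently.
# Each side takes its offset, scale and exponent from the RGB of its own EdgeControl vector input.
# The two results multiply into the Mask output.
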MaterialFunction /Game/Effects/Niagara/MasterMaterials/Functions/Common/FXCommon_EdgeMask_Ex
graph {nodes [
  n1 = FunctionInput() {bUsePreviewValueAsDefault=true, inputName=UorV, inputType=FunctionInput_Scalar, sortPriority=1}
  n2 = FunctionInput() {bUsePreviewValueAsDefault=true, inputName=NoInvert_EdgeControl, sortPriority=3, previewValue=<0 1 1 0>}
  n3 = MaterialFunctionCall(Float3=@n2) {materialFunction=&BreakOutFloat3Components}
  n4 = Saturate(None=Multiply(A=Add(A=@n1, B=@n3.R), B=@n3.G))
  n5 = OneMinus(None=@n1)
  n6 = FunctionInput() {bUsePreviewValueAsDefault=true, inputName=Invert_EdgeControl, sortPriority=3, previewValue=<0 1 1 0>}
  n7 = MaterialFunctionCall(Float3=@n6) {materialFunction=&BreakOutFloat3Components}
  n8 = Saturate(None=Multiply(A=Add(A=@n5, B=@n7.R), B=@n7.G))
  FunctionOutput(None=Multiply(A=Power(Base=@n4, Exp=@n3.B), B=Power(Base=@n8, Exp=@n7.B))) {outputName=Mask}
 ]}
