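# FX material function that builds a simple edge-fade mask along one UV axis, U or V.
# It takes the minimum of the coordinate and its one-minus, giving a symmetric distance from both ends.
# EdgeControl packs the shaping controls: X offsets that distance, Y scales it, and Z is the exponent applied after saturation to produce the Mask output.
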MaterialFunction /Game/Effects/Niagara/MasterMaterials/Functions/Common/FXCommon_EdgeMask_Simple
graph {nodes [
  n1 = FunctionInput() {bUsePreviewValueAsDefault=true, inputName=UorV, inputType=FunctionInput_Scalar, sortPriority=1}
  n2 = OneMinus(None=@n1)
  n3 = FunctionInput() {bUsePreviewValueAsDefault=true, inputName=EdgeControl, sortPriority=3, previewValue=<0 1 1 0>}
  n4 = MaterialFunctionCall(Float3=@n3) {materialFunction=&BreakOutFloat3Components}
  n5 = Multiply(A=Add(A=Min(A=@n1, B=@n2), B=@n4.R), B=@n4.G)
  FunctionOutput(None=Power(Base=Saturate(None=@n5), Exp=@n4.B)) {outputName=Mask}
 ]}
