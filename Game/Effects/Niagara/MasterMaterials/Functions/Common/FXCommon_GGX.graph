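# FXCommon_GGX is a material function that computes the GGX normal distribution term for specular highlights in Niagara effect master materials.
# It normalizes the sum of light and camera vectors into a half vector, then dots it with the normal and saturates.
# Roughness raised to the fourth power and that dot product feed the GGX formula.
# Inputs default to pixel normal, camera vector, sky atmosphere light direction and a preset roughness.
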MaterialFunction /Game/Effects/Niagara/MasterMaterials/Functions/Common/FXCommon_GGX
graph {nodes [
  n1 = FunctionInput() {bUsePreviewValueAsDefault=true, inputName=Roughness, inputType=FunctionInput_Scalar, previewValue=<0.2 0 0 1>}
  n2 = Multiply(A=@n1, B=@n1)
  n3 = Multiply(A=@n2, B=@n2)
  n4 = PixelNormalWS()
  n5 = Add(
   A=FunctionInput(Preview=SkyAtmosphereLightDirection()) {bUsePreviewValueAsDefault=true, inputName=LightDirctionIn}
   B=FunctionInput(Preview=CameraVectorWS()) {bUsePreviewValueAsDefault=true, inputName=CameraVectorIn}
  )
  n6 = Saturate(None=DotProduct(A=FunctionInput(Preview=@n4) {bUsePreviewValueAsDefault=true, inputName=NormalIn}, B=Normalize(VectorInput=@n5)))
  n7 = Multiply(A=@n6, B=@n3)
  n8 = Add(A=Multiply(A=Subtract(A=@n7, B=@n6), B=@n6))
  n9 = Multiply(A=MaterialFunctionCall() {materialFunction=&Pi}.Result, B=@n8)
  FunctionOutput(None=Divide(A=@n3, B=Multiply(A=@n9, B=@n8)))
 ]}
