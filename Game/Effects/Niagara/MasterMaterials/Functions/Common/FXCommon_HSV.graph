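# FXCommon_HSV is a Niagara master material function that applies a hue, saturation and value adjustment to an input color.
# It splits the HSV vector, shifts the color's hue by the first component, desaturates by one minus the second, then multiplies by the third.
# The HSV input defaults to (0, 1, 1), leaving the color unchanged.
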MaterialFunction /Game/Effects/Niagara/MasterMaterials/Functions/Common/FXCommon_HSV
graph {nodes [
  n1 = MaterialFunctionCall(Float3=FunctionInput() {bUsePreviewValueAsDefault=true, inputName=HSV, previewValue=<0 1 1 1>}) {materialFunction=&BreakOutFloat3Components}
  n2 = MaterialFunctionCall(Hue Shift Percentage=@n1.R, Texture=FunctionInput() {inputName=Color}) {materialFunction=&HueShift}
  n3 = Multiply(A=Add(A=@n1.G) {constB=-1}) {constB=-1}
  FunctionOutput(None=Multiply(A=Desaturation(None=@n2.Result, Fraction=@n3), B=@n1.B))
 ]}
