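# FXCommon_NoiseControl is a material function for Niagara effect master materials that offsets and scales a two-component noise value.
# It joins the Control_RGB and Control_A inputs into one four-channel control and breaks out its components.
# Red is added to both noise components, then blue multiplies the sum, so the offset is scaled too.
# Green and alpha do not affect the output.
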MaterialFunction /Game/Effects/Niagara/MasterMaterials/Functions/Common/FXCommon_NoiseControl
graph {nodes [
  n1 = FunctionInput() {inputName=NoiseValue, inputType=FunctionInput_Vector2}
  n2 = FunctionInput() {bUsePreviewValueAsDefault=true, inputName=Control_RGB, sortPriority=1, previewValue=<1 1 0 1>}
  n3 = MaterialFunctionCall(Float4=AppendVector(A=@n2, B=FunctionInput() {bUsePreviewValueAsDefault=true, inputName=Control_A, inputType=FunctionInput_Scalar, sortPriority=2})) {materialFunction=&BreakOutFloat4Components}
  n4 = AppendVector(A=@n3.R, B=@n3.R)
  FunctionOutput(None=Multiply(A=Add(A=@n1, B=@n4), B=AppendVector(A=@n3.B, B=@n3.B))) {outputName=NoiseValue}
 ]}
