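# Material function for FX materials that computes a screen UV offset imitating pixel normal offset refraction.
# The view-space XY difference between the flat and input tangent normals is scaled by refraction strength, which weakens beyond FalloffDistance.
# View size, buffer size and field of view scale the offset.
# Scene depth at the offset UV, with DepthBias, fades it to zero where it would sample foreground objects.
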MaterialFunction /Game/Effects/Niagara/MasterMaterials/Functions/Common/FXCommon_PixelNormalOffsetRefraction
graph {
 properties {bExposeToLibrary=true, previewBlendMode=BLEND_Translucent}
 nodes [
  n1 = Constant3Vector() {constant=<rgba 0 0 1 0>}
  n2 = FunctionInput() {bUsePreviewValueAsDefault=true, inputName=NormalTS, previewValue=<0 0 1 1>}
  n3 = ComponentMask(None=Subtract(A=Transform(None=@n1) {transformType=TRANSFORM_View}, B=Transform(None=@n2) {transformType=TRANSFORM_View})) {g=true, r=true}
  n4 = FunctionInput() {bUsePreviewValueAsDefault=true, inputName=RefractionStrength, inputType=FunctionInput_Scalar, previewValue=<1 0 0 1>}
  n5 = Subtract(A=@n4)
  n6 = FunctionInput() {bUsePreviewValueAsDefault=true, inputName=FalloffDistance, inputType=FunctionInput_Scalar, previewValue=<400 0 0 1>}
  n7 = Clamp(None=Divide(A=PixelDepth().R, B=@n6) {constB=400}) {clampMode=CMODE_ClampMin, minDefault=1}
  n8 = Subtract(A=Add(A=Divide(A=@n5, B=@n7)))
  n9 = ViewProperty() {property=MEVP_ViewSize}
  n10 = Multiply(A=Multiply(A=Multiply(A=@n3, B=@n8), B=@n9.Property), B=ViewProperty() {property=MEVP_BufferSize}.InvProperty)
  n11 = ViewProperty() {property=MEVP_TanHalfFieldOfView}
  n12 = MaterialFunctionCall(Float2=ViewProperty() {property=MEVP_ViewSize}.Property) {materialFunction=&BreakOutFloat2Components}
  n13 = Divide(A=@n12.R, B=@n12.R)
  n14 = Multiply(
   A=Multiply(A=Multiply(A=@n10, B=@n9.Property), B=Constant2Vector() {g=-0.0002, r=0.0002})
   B=MaterialFunctionCall(X=@n11.InvProperty, Y=Multiply(A=@n13, B=@n11.InvProperty)) {materialFunction=&MakeFloat2}.Result
  )
  n15 = Add(A=@n14, B=ScreenPosition().ViewportUV)
  n16 = FunctionInput() {bUsePreviewValueAsDefault=true, inputName=DepthBias, inputType=FunctionInput_Scalar, previewValue=<100 0 0 1>}
  n17 = Subtract(A=Subtract(A=PixelDepth().R, B=SceneDepth(UVs=@n15).R), B=@n16) {constB=100}
  FunctionOutput(None=LinearInterpolate(A=@n14, Alpha=Saturate(None=@n17)) {constB=0}) {outputName=ScreenUVOffsetFraction}
 ]
}
