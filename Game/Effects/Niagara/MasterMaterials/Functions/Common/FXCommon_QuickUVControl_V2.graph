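# Material function for Niagara effect materials that tiles and pans UVs from a packed control input.
# Red sets a uniform scale, applied as a plain multiply or, via a static switch, reciprocally through ScaleUVsByCenter.
# Blue times the time input (default: Time with a long override period) adds an equal U and V scroll.
# The green and alpha controls are split out but unused.
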MaterialFunction /Game/Effects/Niagara/MasterMaterials/Functions/Common/FXCommon_QuickUVControl_V2
graph {nodes [
  n1 = FunctionInput(Preview=TextureCoordinate()) {bUsePreviewValueAsDefault=true, inputName=UVInput, inputType=FunctionInput_Vector2}
  n2 = FunctionInput() {bUsePreviewValueAsDefault=true, inputName=Control_RGB, sortPriority=2, previewValue=<1 1 0 1>}
  n3 = FunctionInput() {bUsePreviewValueAsDefault=true, inputName=Control_A, inputType=FunctionInput_Scalar, sortPriority=3, previewValue=<0 0 0 0>}
  n4 = MaterialFunctionCall(Float4=AppendVector(A=@n2, B=@n3)) {materialFunction=&BreakOutFloat4Components}
  n5 = AppendVector(A=@n4.R, B=@n4.R)
  n6 = MaterialFunctionCall(UVs=@n1, Texture Scale=Divide(B=@n5)) {materialFunction=&ScaleUVsByCenter}
  n7 = Multiply(A=@n1, B=@n5)
  n8 = FunctionInput(Preview=StaticBool()) {bUsePreviewValueAsDefault=true, inputName=ScaleUVByCenter, inputType=FunctionInput_StaticBool, sortPriority=100}
  n9 = FunctionInput(Preview=Time() {bOverride_Period=true, period=13333}) {bUsePreviewValueAsDefault=true, inputName=TimeInput, inputType=FunctionInput_Scalar, sortPriority=1}
  n10 = AppendVector(A=@n4.B, B=@n4.B)
  FunctionOutput(None=Add(A=StaticSwitch(True=@n6.UVs, False=@n7, Value=@n8), B=Multiply(A=@n9, B=@n10))) {outputName=UV}
 ]}
