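# Niagara FX material function that samples a supplied linear-color texture with explicit derivatives.
# Base UV plus an additional offset passes through FXCommon_QuickUVControl, driven by control inputs and overridable periodic time.
# The resulting UV feeds the sample and default derivatives; both derivatives are overridable, though the DDY default also uses DDX.
# Outputs cover RGB, individual channels, processed UV and an RG pair that duplicates R.
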MaterialFunction /Game/Effects/Niagara/MasterMaterials/Functions/Common/FXCommon_TextureSample_DDXDDY
graph {nodes [
  n1 = TextureCoordinate()
  n2 = FunctionInput() {bUsePreviewValueAsDefault=true, inputName=UVControl_RGB, sortPriority=3, previewValue=<1 1 0 1>}
  n3 = MaterialFunctionCall!(
   UVInput=Add(
    A=FunctionInput(Preview=@n1) {bUsePreviewValueAsDefault=true, inputName="[Override]UVInput", inputType=FunctionInput_Vector2, sortPriority=101}
    B=FunctionInput() {bUsePreviewValueAsDefault=true, inputName=AdditionalUVOffset, inputType=FunctionInput_Vector2, sortPriority=2}
   )
   TimeInput=FunctionInput(Preview=Time() {bOverride_Period=true, period=13333}) {bUsePreviewValueAsDefault=true, inputName="[Override]TimeInput", inputType=FunctionInput_Scalar, sortPriority=100}
   Control_RGB=@n2
   Control_A=FunctionInput() {bUsePreviewValueAsDefault=true, inputName=UVControl_A, inputType=FunctionInput_Scalar, sortPriority=4}
  ) {materialFunction=&/FXCommon_QuickUVControl}
  n4 = TextureSample!(
   UVs=@n3.UV
   Tex=FunctionInput() {inputName=Texture, inputType=FunctionInput_Texture2D, sortPriority=-1}
   "DDX(UVs)"=FunctionInput(Preview=DDX(Value=@n3.UV)) {bUsePreviewValueAsDefault=true, inputName="[Override]DDXInput", inputType=FunctionInput_Vector2}
   "DDY(UVs)"=FunctionInput(Preview=DDX(Value=@n3.UV)) {bUsePreviewValueAsDefault=true, inputName="[Override]DDYInput", inputType=FunctionInput_Vector2}
   Apply View MipBias
  ) {mipValueMode=TMVM_Derivative, samplerType=SAMPLERTYPE_LinearColor}
  FunctionOutput(None=@n4.RGB) {outputName=RGB, sortPriority=-1}
  FunctionOutput(None=@n4.A) {outputName=A, sortPriority=4}
  FunctionOutput(None=AppendVector(A=@n4.R, B=@n4.R)) {outputName=RG}
  FunctionOutput(None=@n3.UV) {outputName=UV, sortPriority=5}
  FunctionOutput(None=@n4.R) {outputName=R, sortPriority=1}
  FunctionOutput(None=@n4.B) {outputName=B, sortPriority=3}
  FunctionOutput(None=@n4.G) {outputName=G, sortPriority=2}
 ]}
