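# Library-exposed material function for Niagara effect master materials that selects one of five scalar inputs by shading path and quality level.
# A shading path switch feeds two quality switches, one for default shading and one for mobile.
# Default shading has separate Low and High inputs, with other levels using the PC-High/Epic/current-gen console input.
# Mobile uses the PC-Perf/Android input at Low and the Switch input otherwise.
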MaterialFunction /Game/Effects/Niagara/MasterMaterials/Functions/Common/PlatformSwitch_1D
graph {
 properties {bExposeToLibrary=true}
 nodes [
  n1 = FunctionInput() {bUsePreviewValueAsDefault=true, inputName="PC-Low(NotPerf)&&Medium", inputType=FunctionInput_Scalar}
  n2 = FunctionInput() {bUsePreviewValueAsDefault=true, inputName="Xb1,PS4", inputType=FunctionInput_Scalar}
  n3 = FunctionInput() {bUsePreviewValueAsDefault=true, inputName="PC-High&&Epic,XSX,PS5", inputType=FunctionInput_Scalar}
  n4 = FunctionInput() {bUsePreviewValueAsDefault=true, inputName=Switch, inputType=FunctionInput_Scalar}
  n5 = FunctionInput() {bUsePreviewValueAsDefault=true, inputName="PC-Perf,Android", inputType=FunctionInput_Scalar}
  FunctionOutput(None=ShadingPathSwitch(Default=QualitySwitch(Default=@n3, Low=@n1, High=@n2, Epic=@n3), Mobile=QualitySwitch(Default=@n4, Low=@n5, Medium=@n4)))
 ]
}
