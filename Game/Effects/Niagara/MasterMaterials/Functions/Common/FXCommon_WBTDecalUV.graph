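# Material function that builds decal-style UVs and masks on a translucent mesh by projecting the world position behind it into instance space.
# That position, scaled by the mesh bound size input, yields planar and radial UVs, their tiling-method 0-1 masks, and a local height mask.
# It also outputs a pixel-versus-scene depth mask and a thresholded custom depth mask, which is constant 1 on Switch, Android and PC performance mode.
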MaterialFunction /Game/Effects/Niagara/MasterMaterials/Functions/Common/FXCommon_WBTDecalUV
graph {
 properties {bExposeToLibrary=true, previewBlendMode=BLEND_Translucent}
 nodes [
  n1 = Constant3Vector() {constant=<rgba 100 100 100 0>}
  n2 = Divide(
   A=TransformPosition(None=MaterialFunctionCall() {materialFunction=&WorldPositionBehindTranslucency}.Result) {transformSourceType=TRANSFORMPOSSOURCE_World, transformType=TRANSFORMPOSSOURCE_Instance}
   B=Multiply(
    A=FunctionInput(Preview=@n1) {bUsePreviewValueAsDefault=true, inputName=MeshOriginalBoundSize}
    B=Constant3Vector() {constant=<rgba 1 1 0.5 0>}
   )
  )
  Abs(None=@n2)
  n3 = Add(A=@n2) {constB=0.5}
  FunctionOutput(None=ComponentMask(None=@n3) {g=true, r=true}) {outputName=UV}
  n4 = MaterialFunctionCall(Vector or UVs=ComponentMask(None=Multiply(A=@n2) {constB=2}) {g=true, r=true}) {materialFunction=&VectorToRadialValue}
  FunctionOutput(None=@n4.Radial Coordinates) {outputName=RadialUV, sortPriority=1}
  n5 = Constant2Vector() {g=1, r=1}
  FunctionOutput(None=MaterialFunctionCall!(TexCoord=@n3, ClampChannel=@n5) {materialFunction=/Game/Effects/Fort_Effects/MaterialFunctions/MF_FX_TilingMethod}.Result) {outputName=UV01Mask, sortPriority=2}
  FunctionOutput(None=MaterialFunctionCall(TexCoord=@n4.Radial Coordinates, ClampChannel=@n5) {materialFunction=&MF_FX_TilingMethod}.Result) {outputName=RadialUV01Mask, sortPriority=3}
  n6 = SceneTexture() {sceneTextureId=PPI_CustomDepth}
  n7 = Saturate(None=Floor(None=Divide(A=@n6.Color) {constB=40960}))
  n8 = Constant() {r=1}
  FunctionOutput(None=MaterialFunctionCall("Xb1,PS4"=@n7, "PC-High&&Epic,XSX,PS5"=@n7, "PC-Perf,Android"=@n8, Switch=@n8, "PC-Low(NotPerf)&&Medium"=@n7) {materialFunction=&PlatformSwitch_1D}.Result) {outputName="CustomDepthMask(1onMobile)", sortPriority=3}
  n9 = ComponentMask(None=@n2) {b=true}
  FunctionOutput(None=OneMinus(None=Saturate(None=@n9))) {outputName=ZMask, sortPriority=3}
  n10 = Subtract(A=Subtract(A=PixelDepth().R, B=SceneDepth().R)) {constB=10}
  FunctionOutput(None=Saturate(None=Divide(A=@n10) {constB=10})) {outputName=DepthMask, sortPriority=3}
 ]
}
