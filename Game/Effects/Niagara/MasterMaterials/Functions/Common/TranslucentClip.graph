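# TranslucentClip is a library-exposed material function in the Niagara master materials that discards low-opacity pixels.
# A scalar MaskIn feeds a Custom HLSL node that calls clip(-1) when the value falls below a fixed threshold of about one third.
# The node always returns zero through the ToNormalOrEmissive output, named for the Normal or Emissive channel, so its only effect is the clip.
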MaterialFunction /Game/Effects/Niagara/MasterMaterials/Functions/Common/TranslucentClip
graph {
 properties {bExposeToLibrary=true}
 nodes [
  MaterialExpressionFunctionInput_1 = FunctionInput() {bUsePreviewValueAsDefault=true, inputName=MaskIn, inputType=FunctionInput_Scalar, previewValue=<1 0 0 1>}
  n1 = Custom!(Opacity=@MaterialExpressionFunctionInput_1) {
   code="float mask = 0;\r\nif (Opacity <0.3333 )\r\n{\r\nclip(-1);\r\n}\r\nreturn mask;"
   outputType=CMOT_Float1
   inputs [{
     inputName=Opacity
     input {expression=&self:MaterialExpressionFunctionInput_1}
    }]
  }
  FunctionOutput(None=@n1) {outputName=ToNormalOrEmissive}
 ]
}
